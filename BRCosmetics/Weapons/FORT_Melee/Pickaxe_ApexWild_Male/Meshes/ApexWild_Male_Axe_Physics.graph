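# Physics asset for the ApexWild male two-handed axe pickaxe, providing capsule bodies and joints for the handle and its dynamic cable bones.
# Cable bones 1 to 3 form a chain from the handle, while cable 4 is constrained directly to the handle as a separate branch.
# Every joint locks twist and sets narrow per-joint swing angles, with unequal swing axes on cable 4.
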PhysicsAsset /BRCosmetics/Weapons/FORT_Melee/Pickaxe_ApexWild_Male/Meshes/ApexWild_Male_Axe_Physics
bodies {
 Melee_TwoHanded_Handle [{shapeType=Capsule, length=15.75, radius=5.5, shapeName=Melee_TwoHanded_Handle_capsule, center=<0 0 25.5564>, rotation=<rot -0.0001 0 0.7707>}]
 dyn_cable_1 [{
   shapeType=Capsule
   length=1.3285
   radius=1.2858
   shapeName=dyn_cable_1_capsule
   center=<0.5331 0.0037 -0.041>
   rotation=<rot 0 88.6896 -86.8276>
  }]
 dyn_cable_2 [{shapeType=Capsule, length=2, radius=1, shapeName=dyn_cable_2_capsule, rotation=<rot 75.1263 -179.9998 -179.9999>}]
 dyn_cable_3 [{shapeType=Capsule, length=2, radius=1, shapeName=dyn_cable_3_capsule, rotation=<rot 70.9558 0 -0.0001>}]
 dyn_cable_4 [{shapeType=Capsule, length=1.071, radius=0.6971, shapeName=dyn_cable_4_capsule, center=<-0.2703 0 0.065>, rotation=<rot 76.4851 0 0>}]
}
constraints [
 {
  bone1Name=dyn_cable_1
  bone2Name=Melee_TwoHanded_Handle
  swing1LimitDegrees=10
  swing2LimitDegrees=10
  twistLimitDegrees=45
  twistMotion=Locked
 }
 {bone1Name=dyn_cable_2, bone2Name=dyn_cable_1, swing1LimitDegrees=17, swing2LimitDegrees=17, twistLimitDegrees=45, twistMotion=Locked}
 {bone1Name=dyn_cable_3, bone2Name=dyn_cable_2, swing1LimitDegrees=6, swing2LimitDegrees=6, twistLimitDegrees=45, twistMotion=Locked}
 {
  bone1Name=dyn_cable_4
  bone2Name=Melee_TwoHanded_Handle
  swing1LimitDegrees=32.28
  swing2LimitDegrees=4.92
  twistLimitDegrees=45
  twistMotion=Locked
 }
]
